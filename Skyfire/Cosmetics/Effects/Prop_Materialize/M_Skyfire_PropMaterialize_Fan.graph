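# Translucent, unlit, two-sided material for Niagara mesh particles, the fan element of Skyfire's prop materialize effect.
# Dot and soft-noise textures pan on the second UV channel and are tinted bright cyan-blue, then multiplied by particle color for emissive.
# The same layers form opacity, shaped by vertical and horizontal gradient masks, Fresnel and particle alpha.
# The Fresnel takes its normal from a SplineThicken function.
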Material /Skyfire/Cosmetics/Effects/Prop_Materialize/M_Skyfire_PropMaterialize_Fan
graph {
 properties {
  bUsedWithNiagaraMeshParticles=true
  blendMode=BLEND_Translucent
  pixelDepthOffsetMode=PDOM_Legacy
  shadingModel=MSM_Unlit
  twoSided=true
  usedShadingModels=Unlit
 }
 nodes [
  n1 = ParticleColor()
  n2 = Constant() {r=1.5}
  n3 = Constant() {r=2.5}
  n4 = Panner(
   Coordinate=Multiply(A=TextureCoordinate() {coordinateIndex=1}, B=AppendVector(A=@n2, B=@n3))
   Speed=AppendVector(A=Constant() {r=-2}, B=Constant() {r=0.2})
  )
  n5 = Multiply(
   A=Power(Base=TextureSample(UVs=@n4) {texture=&/T_Skyfire_Dots}.RGB, Exp=Constant() {r=1})
   B=Constant() {r=1}
  )
  n6 = Multiply(A=@n5, B=Constant3Vector() {constant=<rgba 0 5.0425 10 1>})
  n7 = TextureCoordinate() {coordinateIndex=1}
  n8 = AppendVector(A=Constant() {r=0.5}, B=Constant() {r=2})
  n9 = Constant() {r=-1}
  n10 = Constant() {r=0.1}
  n11 = TextureSample(UVs=Panner(Coordinate=Multiply(A=@n7, B=@n8), Speed=AppendVector(A=@n9, B=@n10))) {texture=/Game/Characters/Player/Male/Medium/Bodies/M_MED_Viper/Textures/Glitch/T_SoftNoise}
  n12 = Constant() {r=1}
  n13 = Clamp(None=Multiply(A=Power(Base=@n11.B, Exp=@n12), B=Constant() {r=1}))
  n14 = Multiply(A=Constant3Vector() {constant=<rgba 0 0.9171 2 1>}, B=@n13)
  MaterialFunctionCall(Emissive=Multiply(A=@n1.RGB, B=Add(A=@n6, B=@n14))) {materialFunction=&MF_FX_EmissiveAdjustments}
  n15 = Multiply(
   A=Add(A=Clamp(None=@n5), B=@n13)
   B=TextureSample() {samplerType=SAMPLERTYPE_Masks, texture=/Game/Global/Textures/Default/Masks_and_Gradients/T_VertGradient_MASK_2}.RGB
  )
  n16 = Power(Base=TextureSample() {texture=/Game/Gadgets/Assets/Misty_Bop/Effects/T_MediumGradient_Horizontal}.RGB, Exp=Constant() {r=3})
  n17 = Constant() {r=1}
  n18 = Constant() {r=1}
  n19 = Constant() {r=10}
  n20 = MaterialFunctionCall(WidthBase=@n19, WidthTip=@n19) {materialFunction=&SplineThicken}
  Multiply(
   A=@n1.A
   B=Multiply(
    A=Multiply(A=@n15, B=@n16)
    B=MaterialFunctionCall(Brightness=@n17, Exponent=@n18, Vertex / Pixel Normal=@n20.Normal) {materialFunction=&MF_Fresnel_FX}.Fresnel
   )
  )
 ]
}
