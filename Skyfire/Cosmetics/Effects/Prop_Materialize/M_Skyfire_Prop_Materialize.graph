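# Masked, two-sided lit material for a Skyfire prop materialize effect, with usage enabled for Niagara particles and skeletal meshes.
# A world-space sphere mask at an offset from particle position grows with the Erode dynamic parameter.
# Subtracting a smaller sphere, inset by EdgeThickness, yields a particle-colored edge band.
# Back faces add cyan via TwoSidedSign, and an extra emissive texture joins before the emissive adjustment function.
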Material /Skyfire/Cosmetics/Effects/Prop_Materialize/M_Skyfire_Prop_Materialize
graph {
 properties {
  bUsedWithMeshDeformer=true
  bUsedWithNiagaraMeshParticles=true
  bUsedWithNiagaraRibbons=true
  bUsedWithNiagaraSprites=true
  bUsedWithSkeletalMesh=true
  blendMode=BLEND_Masked
  pixelDepthOffsetMode=PDOM_Legacy
  twoSided=true
  usedShadingModels=Default Lit
 }
 nodes [
  n1 = TextureSample(UVs=Multiply(
    A=MaterialFunctionCall() {materialFunction=&ScreenAlignedUVs}."X Scale to Ratio, Y 100%"
    B=Constant() {r=30}
   )) {texture=/Game/Global/Textures/Default/Masks_and_Gradients/16x}
  n2 = Constant3Vector() {constant=<rgba 1 1 1 0>}
  n3 = VectorParameter() {parameterName=Diffuse_Overlay, defaultValue=<rgba 0 0.5043 1 1>}
  Add(A=Add(A=Multiply(A=@n1.RGB, B=@n2), B=Multiply(A=@n3.RGB)))
  n4 = Constant() {r=1}
  n5 = Constant() {r=5}
  Add(A=Multiply(
    A=MaterialFunctionCall(Brightness=@n4, Exponent=@n5) {materialFunction=&MF_Fresnel_FX}.Fresnel Inverted
    B=Constant3Vector() {constant=<rgba 1 1 1 0>}
   ))
  n6 = Time()
  Clamp(None=ConstantBiasScale(None=Sine(None=@n6)))
  n7 = TextureCoordinate()
  n8 = DynamicParameter() {defaultValue=<rgba 1 0.5 1 1>, paramNames [Erode_0-1, EdgeThickness, Param3, Param4]}
  n9 = Multiply(A=Clamp(None=@n8.Erode_0-1), B=ScalarParameter() {defaultValue=35, group=Sphere Mask, parameterName=Erode_Radius})
  n10 = ScalarParameter() {defaultValue=1, group=Sphere Mask, parameterName=Hardness}
  n11 = Add(
   A=ParticlePositionWS()
   B=Transform(None=VectorParameter() {parameterName=Offset_Spheremask_Position, defaultValue=<rgba 0 -20 0 1>}.RGB) {transformSourceType=TRANSFORMSOURCE_Local}
  )
  n12 = WorldPosition()
  n13 = SphereMask(A=@n12.XYZ, B=@n11, Radius=@n9, Hardness=@n10)
  n14 = Subtract(A=@n9, B=@n8.EdgeThickness)
  n15 = Multiply(A=ParticleColor().RGB, B=Subtract(A=@n13, B=SphereMask(A=@n12.XYZ, B=@n11, Radius=@n14, Hardness=@n10)))
  n16 = Add(A=@n15, B=Constant3Vector() {constant=<rgba 0 0.5043 1 1>})
  n17 = TwoSidedSign()
  MaterialFunctionCall(Emissive=Add(
    A=LinearInterpolate(A=@n16, B=@n15, Alpha=@n17)
    B=TextureSampleParameter2D(UVs=@n7) {parameterName="Extra Emissive ", texture=/Skyfire/Gameplay/KnockbackWeapon/Textures/T_Skyfire_Alien_Knockback_Weapon_E}.RGB
   )) {materialFunction=&MF_FX_EmissiveAdjustments}
  Reroute(None=@n13)
  n18 = VectorParameter() {group=Trailing Edge Sphere Mask, parameterName=Trailing Edge Sphereical UV Scale, sortPriority=4, defaultValue=<rgba 1 1 0 0>}
  n19 = VectorParameter() {group=Trailing Edge Sphere Mask, parameterName="Trailing Edge Sphere Mask Coord (B)", sortPriority=3, defaultValue=<rgba 0.5 0.5 0 0>}
  n20 = Sine(None=Time())
  SphereMask(
   A=Multiply(A=TextureCoordinate(), B=ComponentMask(None=@n18.RGB) {g=true, r=true})
   B=ComponentMask(None=@n19.RGB) {g=true, r=true}
   Radius=Clamp(None=ConstantBiasScale(None=@n20))
   Hardness=Constant() {r=1}
  )
  TextureSampleParameter2D(UVs=TextureCoordinate()) {parameterName=DiffuseTexture, texture=/Game/Global/Textures/Default/Blanks/T_WhiteUserInterface}
  TextureSampleParameter2D(UVs=TextureCoordinate()) {parameterName=SpecularTexture, texture=/Game/Environments/Props/Transylvania/Trees/Textures/Specular01}
  TextureSampleParameter2D(UVs=TextureCoordinate()) {parameterName=Normal Texture, samplerType=SAMPLERTYPE_Normal, texture=&FlatNormalBlue}
  TextureSampleParameter2D(UVs=TextureCoordinate()) {parameterName=Diffuse_Texture, texture=&T_WhiteUserInterface}
 ]
}
