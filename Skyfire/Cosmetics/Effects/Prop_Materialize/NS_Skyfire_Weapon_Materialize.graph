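# Niagara system for a blue, roughly one-second weapon materialize effect with a warmup followed by a burst.
# Weapon_Mesh ramps a dynamic material parameter while glow, light and spinning dust fans sweep along Y on shared curve shapes, alongside traveling waveform-colored rings.
# At 0.8 seconds flares, lights, rings, sparkles and a decal pulse fire.
# User.Opacity fades some layers; the spiral and Flare_Warmup emitters are disabled.
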NiagaraSystem /Skyfire/Cosmetics/Effects/Prop_Materialize/NS_Skyfire_Weapon_Materialize
dependencies {
 usedDataInterfaces [&NiagaraDataInterfaceCurve, &NiagaraDataInterfaceColorCurve]
 usedDynamicInputs []
 usedModules [
  &self:SystemSpawnScript.SystemScriptSource.SystemScriptGraph.NiagaraNodeAssignment_0.SetVariables_#1
  &SystemState
 ]
}
emitters {
 AirGlow001 {
  emitterUpdateScript [
   &EmitterState(Life Cycle Mode=::NewEnumerator1, Loop Behavior=::NewEnumerator1, Loop Delay=0.2, Loop Duration=0.5, UseLoopDelay=true)
   &SpawnBurst_Instantaneous(Spawn Count=1, Use Spawn Probability=false) {enabled=false}
   &SpawnRate(SpawnRate=20, Use Spawn Probability=false)
  ]
  particleSpawnScript [/Niagara/Modules/Spawn/Initialization/InitializeParticle(
    Color=<rgba 0 0.5043 1 1>
    Lifetime=0.1
    Module.Write Color=true
    Module.Write Lifetime=true
    Module.Write Mass=false
    Module.Write Material Random=true
    Module.Write Position=false
    Module.Write Scale=false
    Module.Write SpriteRotation=true
    Module.Write SpriteSize=true
    Position=&SimulationPosition()
    Sprite Rotation=&RandomRangeFloat(Minimum=0, Maximum=360)
    Sprite Size=&Vector2DFromFloat(Value=60)
   )]
  particleUpdateScript [
   &ParticleState()
   /SetVariables(Particles.Position=&MakeVector(
     X=2.5
     Y=&FloatFromCurve(FloatCurve=DataInterface() {propertyValues {Curve {keys [<at 0.2 5>,<at 0.8 70>]}}}, CurveIndex=$Emitter.Age)
     Z=9
    )) {isSetParameters=true, module=SetVariables_#2}
   &DynamicMaterialParameters(
    Index 0 Param 1=2
    Module.Param0WriteEnabled=true
    Module.Param1WriteEnabled=true
    Module.Param2WriteEnabled=true
    Module.Param3WriteEnabled=true
   ) {enabled=false}
   &ScaleColor(
    Scale Alpha=&FloatFromCurve(FloatCurve=DataInterface() {propertyValues {Curve {keys [<at 0 0>,<at 0.1 1>]}}}, CurveIndex=$Particles.Age)
    ScaleA=true
    ScaleRGB=true
   )
  ]
  rendererClasses [&NiagaraSpriteRendererProperties]
  renderers [
   {rendererClass=&NiagaraSpriteRendererProperties}
   {rendererClass=&NiagaraSpriteRendererProperties, rendererIndex=1}
  ]
 }
 Decal_Warmup001 {
  emitterUpdateScript [
   &EmitterState(Life Cycle Mode=::NewEnumerator1, Loop Behavior=::NewEnumerator1, Loop Delay=0.3, Loop Duration=1, UseLoopDelay=true)
   &SpawnBurst_Instantaneous(Spawn Count=1, Use Spawn Probability=false)
  ]
  particleSpawnScript [
   /Niagara/Modules/Spawn/Initialization/V2/InitializeParticle(
    Color=<rgba 3.4367 4.0775 6 1>
    Lifetime=0.7
    Sprite Size Mode=::NewEnumerator3
    Uniform Sprite Size=50
    UsePositionOffset=false
    Write Material Random=true
   )
   &CameraOffset(Camera Offset Amount=20) {enabled=false}
  ]
  particleUpdateScript [
   &ParticleState()
   &ScaleColor(
    Linear Color Curve=DataInterface() {propertyValues {
      AlphaCurve {keys [<at 0 0 RCIM_Cubic>, {arriveTangent=-1.373, interpMode=RCIM_Cubic, leaveTangent=-1.373, tangentMode=RCTM_User, time=0.3, value=1}, <at 1 0>]}
      BlueCurve {keys [<at 0 1>,<at 0.6831 1>]}
      GreenCurve {keys [<at 0 0.3636>,<at 0.6831 0>]}
      RedCurve {keys [<at 0 0>,<at 0.6831 0.4834>]}
     }}
    Scale Mode=::NewEnumerator2
   )
   &DynamicMaterialParameters(
    Index 0 Param 1=1
    Index 0 Param 2=1
    Index 0 Param 3=&FloatFromCurve(FloatCurve=DataInterface() {propertyValues {Curve {keys [{arriveTangent=1.3198, interpMode=RCIM_Cubic, leaveTangent=1.3198, tangentMode=RCTM_User, time=0, value=-0.25}]}}})
    Index 0 Param 4=1
    Module.Param0WriteEnabled=true
    Module.Param1WriteEnabled=true
    Module.Param2WriteEnabled=true
    Module.Param3WriteEnabled=true
   )
  ]
  rendererClasses [&NiagaraMeshRendererProperties]
  renderers [{rendererClass=&NiagaraMeshRendererProperties, rendererIndex=0}]
 }
 DustFan {
  emitterUpdateScript [
   &EmitterState(Life Cycle Mode=::NewEnumerator1, Loop Behavior=::NewEnumerator1, Loop Delay=0.2, Loop Duration=1, UseLoopDelay=true)
   &SpawnBurst_Instantaneous(Spawn Count=2, Use Spawn Probability=false)
  ]
  particleSpawnScript [
   &FN_SetSpawnGroup(Number of Spawn Groups=2)
   &InitializeParticle(
    Color Mode=::NewEnumerator0
    Lifetime=0.6
    Mesh Scale=<0.9 -3 0.9>
    Mesh Scale Mode=::NewEnumerator4
    UsePositionOffset=false
    Write Material Random=true
   )
   &SystemLocation(MaskSpawnGroup=false)
  ]
  particleUpdateScript [
   &ParticleState()
   /SetVariables(Particles.Position=&MakeVector(
     X=2.5
     Y=&FloatFromCurve(FloatCurve=DataInterface() {propertyValues {Curve {keys [<at 0.2 5>,<at 0.8 70>]}}}, CurveIndex=$Emitter.Age)
     Z=9
    )) {isSetParameters=true, module=SetVariables_#3}
   /SetVariables(Particles.Pivot:Vector3f=&Add_Vector(A=<0 0 0>, B=<0 0 0>)) {isSetParameters=true, module=SetVariables_#4}
   /SetVariables(Particles.MeshOrientation=&MultiplyQuaternion(
     Quaternion A=&MakeQuatFromAxisAngle(
      Axis=<0 1 0>
      AngleInDegrees=&FloatFromCurve(FloatCurve=DataInterface() {propertyValues {Curve {keys [<at 0 0>,<at 1 -360>]}}})
     )
     Quaternion B=&MakeQuatFromAxisAngle(
      Axis=<0 1 0>
      AngleInDegrees=&Multiply_Float(A=&MakeFloatFromInt(Int=&SpawnGroup()), B=180)
     )
    )) {isSetParameters=true, module=SetVariables_#5}
   &ScaleColor(
    Scale Alpha=&FloatFromCurve(FloatCurve=DataInterface() {propertyValues {Curve {keys [<at 0 0>,<at 0.5 1>,<at 1 0>]}}})
    Scale RGB=&VectorFromFloat(Value=&FloatFromCurve(FloatCurve=DataInterface() {propertyValues {Curve {keys [<at 0 0>,<at 0.5 1>,<at 1 0.1>]}}}))
    ScaleA=true
    ScaleRGB=true
   )
  ]
  rendererClasses [&NiagaraMeshRendererProperties]
  renderers [{rendererClass=&NiagaraMeshRendererProperties, rendererIndex=0}]
 }
 DustFan001 {
  emitterUpdateScript [
   &EmitterState(Life Cycle Mode=::NewEnumerator1, Loop Behavior=::NewEnumerator1, Loop Delay=0.2, Loop Duration=1, UseLoopDelay=true)
   &SpawnBurst_Instantaneous(Spawn Count=2, Use Spawn Probability=false)
  ]
  particleSpawnScript [
   &FN_SetSpawnGroup(Number of Spawn Groups=2)
   &InitializeParticle(
    Color Mode=::NewEnumerator0
    Lifetime=0.6
    Mesh Scale=<0.9 1 0.9>
    Mesh Scale Mode=::NewEnumerator4
    UsePositionOffset=false
    Write Material Random=true
   )
   &SystemLocation(MaskSpawnGroup=false)
  ]
  particleUpdateScript [
   &ParticleState()
   /SetVariables(Particles.Position=&MakeVector(
     X=2.5
     Y=&FloatFromCurve(FloatCurve=DataInterface() {propertyValues {Curve {keys [<at 0.2 5>,<at 0.8 70>]}}}, CurveIndex=$Emitter.Age)
     Z=9
    )) {isSetParameters=true, module=SetVariables_#3}
   /SetVariables(Particles.Pivot:Vector3f=&Add_Vector(A=<0 0 0>, B=<0 0 0>)) {isSetParameters=true, module=SetVariables_#4}
   /SetVariables(Particles.MeshOrientation=&MultiplyQuaternion(
     Quaternion A=&MakeQuatFromAxisAngle(
      Axis=<0 1 0>
      AngleInDegrees=&FloatFromCurve(FloatCurve=DataInterface() {propertyValues {Curve {keys [<at 0 0>,<at 1 -180>]}}})
     )
     Quaternion B=&MakeQuatFromAxisAngle(
      Axis=<0 1 0>
      AngleInDegrees=&Multiply_Float(A=&MakeFloatFromInt(Int=&SpawnGroup()), B=180)
     )
    )) {isSetParameters=true, module=SetVariables_#5}
   &ScaleColor(
    Scale Alpha=&FloatFromCurve(FloatCurve=DataInterface() {propertyValues {Curve {keys [<at 0 0>,<at 0.5 1>,<at 1 0>]}}})
    Scale RGB=&VectorFromFloat(Value=&FloatFromCurve(FloatCurve=DataInterface() {propertyValues {Curve {keys [<at 0 0>,<at 0.5 1>,<at 1 0.1>]}}}))
    ScaleA=true
    ScaleRGB=true
   )
  ]
  rendererClasses [&NiagaraMeshRendererProperties]
  renderers [{rendererClass=&NiagaraMeshRendererProperties, rendererIndex=0}]
 }
 Flare_Burst {
  emitterUpdateScript [
   &EmitterState(Life Cycle Mode=::NewEnumerator1, Loop Behavior=::NewEnumerator1, Loop Duration=1, UseLoopDelay=false)
   &SpawnBurst_Instantaneous(Spawn Count=1, Spawn Time=0.8, Use Spawn Probability=false)
   &SpawnBurst_Instantaneous(Spawn Count=1, Spawn Time=0.95, Use Spawn Probability=false) {module=SpawnBurst_Instantaneous001}
  ]
  particleSpawnScript [
   /Niagara/Modules/Spawn/Initialization/InitializeParticle(
    Color=<rgba 0 0.0714 0.5 0.5>
    Lifetime=0.2
    Module.Write Color=true
    Module.Write Lifetime=true
    Module.Write Mass=true
    Module.Write Material Random=true
    Module.Write Position=false
    Module.Write Scale=false
    Module.Write SpriteRotation=false
    Module.Write SpriteSize=true
    Position=&SimulationPosition()
    Sprite Size=<30 30>
   )
   &CameraOffset(Camera Offset Amount=20)
   &SystemLocation(MaskSpawnGroup=false, Offset=<0 40 10>)
  ]
  particleUpdateScript [
   &ParticleState()
   &ScaleSpriteSize(Scale Factor=&Vector2DFromFloat(Value=&RandomRangeFloat(Minimum=3.5, Maximum=5, Evaluation Type=::NewEnumerator1)))
   &ScaleColor(Scale Alpha=$User.Opacity, ScaleA=true, ScaleRGB=true)
   &FN_Light_Attributes(Color=<rgba 0 5.0425 10 1>, Create Color Attribute=true, Create Exponent Attribute=true, Create Radius Attribute=true, Exponent=3, Radius=50)
  ]
  rendererClasses [&NiagaraSpriteRendererProperties, &NiagaraLightRendererProperties]
  renderers [
   {rendererClass=&NiagaraSpriteRendererProperties}
   {rendererClass=&NiagaraSpriteRendererProperties, rendererIndex=1}
   {rendererClass=&NiagaraLightRendererProperties, rendererIndex=2}
  ]
 }
 Flare_Warmup {
  bEnabled=false
  emitterUpdateScript [
   &EmitterState(Life Cycle Mode=::NewEnumerator1, Loop Behavior=::NewEnumerator1, Loop Duration=0.8, UseLoopDelay=false)
   &SpawnRate(SpawnRate=10, Use Spawn Probability=false)
  ]
  particleSpawnScript [
   &InitializeParticle(
    Color=<rgba 0 0.1429 1 1>
    Lifetime=0.2
    Module.Write Color=true
    Module.Write Lifetime=true
    Module.Write Mass=true
    Module.Write Material Random=true
    Module.Write Position=false
    Module.Write Scale=false
    Module.Write SpriteRotation=true
    Module.Write SpriteSize=true
    Position=&SimulationPosition()
    Sprite Rotation=&RandomRangeFloat(Minimum=0, Maximum=360)
    Sprite Size=&Vector2DFromFloat(Value=&FloatFromCurve(FloatCurve=DataInterface() {propertyValues {Curve {keys [<at 0 5>,<at 0.8 20>]}}}, CurveIndex=$Emitter.Age))
   )
   &CameraOffset()
   &SystemLocation(MaskSpawnGroup=false, Offset=<0 40 10>)
  ]
  particleUpdateScript [
   &ParticleState()
   &ScaleSpriteSize(Scale Factor=&Vector2DFromFloat(Value=&RandomRangeFloat(Minimum=3.5, Maximum=5, Evaluation Type=::NewEnumerator1)))
   &ScaleColor(Scale Alpha=$User.Opacity, ScaleA=true, ScaleRGB=true)
  ]
  rendererClasses [&NiagaraSpriteRendererProperties]
  renderers [
   {rendererClass=&NiagaraSpriteRendererProperties}
   {rendererClass=&NiagaraSpriteRendererProperties, rendererIndex=1}
  ]
 }
 Light {
  emitterUpdateScript [
   &EmitterState(Life Cycle Mode=::NewEnumerator1, Loop Behavior=::NewEnumerator1, Loop Delay=0.2, Loop Duration=0.5, UseLoopDelay=true)
   &SpawnBurst_Instantaneous(Spawn Count=1, Use Spawn Probability=false)
   &SpawnRate(SpawnRate=20, Use Spawn Probability=false) {enabled=false}
  ]
  particleSpawnScript [
   &InitializeParticle(
    Color=<rgba 0 10.085 20 1>
    Lifetime=0.6
    Module.Write Color=true
    Module.Write Lifetime=true
    Module.Write Mass=false
    Module.Write Material Random=true
    Module.Write Position=false
    Module.Write Scale=false
    Module.Write SpriteRotation=true
    Module.Write SpriteSize=true
    Position=&SimulationPosition()
    Sprite Rotation=&RandomRangeFloat(Minimum=0, Maximum=360)
    Sprite Size=<15 15>
   )
   /SetVariables(Particles.Flare_Size:Vector2f=&Vector2DFromFloat(Value=15)) {isSetParameters=true}
   /SetVariables(Particles.SpriteAlignment=<0 0 1>, Particles.SpriteFacing=<0 1 0>) {isSetParameters=true}
  ]
  particleUpdateScript [
   &ParticleState()
   /SetVariables(Particles.Position=&MakeVector(
     X=2.5
     Y=&FloatFromCurve(FloatCurve=DataInterface() {propertyValues {Curve {keys [<at 0.2 5>,<at 0.8 70>]}}}, CurveIndex=$Emitter.Age)
     Z=9
    )) {isSetParameters=true, module=SetVariables_#2}
   &DynamicMaterialParameters(
    Index 0 Param 1=2
    Module.Param0WriteEnabled=true
    Module.Param1WriteEnabled=true
    Module.Param2WriteEnabled=true
    Module.Param3WriteEnabled=true
   ) {enabled=false}
   &ScaleColor(
    Scale Alpha=&FloatFromCurve(FloatCurve=DataInterface() {propertyValues {Curve {keys [<at 0 0>,<at 0.1 1>]}}}, CurveIndex=$Particles.Age)
    ScaleA=true
    ScaleRGB=true
   )
   &FN_Light_Attributes(Color=<rgba 0 5.0425 10 1>, Create Color Attribute=true, Create Exponent Attribute=true, Create Radius Attribute=true, Exponent=3, Radius=20)
  ]
  rendererClasses [&NiagaraSpriteRendererProperties, &NiagaraLightRendererProperties]
  renderers [
   {rendererClass=&NiagaraSpriteRendererProperties}
   {rendererClass=&NiagaraSpriteRendererProperties, rendererIndex=1}
   {rendererClass=&NiagaraLightRendererProperties, rendererIndex=2}
  ]
 }
 NE_Glow002 {
  emitterUpdateScript [
   &EmitterState(Life Cycle Mode=::NewEnumerator1, Loop Behavior=::NewEnumerator1, Loop Duration=0.8, UseLoopDelay=false)
   &SpawnRate(SpawnRate=10, Use Spawn Probability=false)
  ]
  particleSpawnScript [
   &InitializeParticle(
    Color=<rgba 0 0.1429 1 1>
    Lifetime=0.2
    Module.Write Color=true
    Module.Write Lifetime=true
    Module.Write Mass=true
    Module.Write Material Random=true
    Module.Write Position=false
    Module.Write Scale=false
    Module.Write SpriteRotation=true
    Module.Write SpriteSize=true
    Position=&SimulationPosition()
    Sprite Rotation=&RandomRangeFloat(Minimum=0, Maximum=360)
    Sprite Size=&Vector2DFromFloat(Value=&RandomRangeFloat(Minimum=1, Maximum=2))
   )
   &CameraOffset(Camera Offset Amount=20)
   &SystemLocation(MaskSpawnGroup=false, Offset=<0 40 10>)
   &SphereLocation(
    MaskSpawnGroup=false
    Non Uniform Scale=<1 4 1>
    Override Seed=false
    Sphere Radius=10
    Surface Only=false
    UseNonUniformScale=true
    UseOffset=false
    UseSphereOrientation=false
   )
  ]
  particleUpdateScript [
   &ParticleState()
   &ScaleSpriteSize(Scale Factor=&Vector2DFromFloat(Value=&RandomRangeFloat(Minimum=3.5, Maximum=5, Evaluation Type=::NewEnumerator1)))
   &ScaleColor(Scale Alpha=$User.Opacity, ScaleA=true, ScaleRGB=true)
  ]
  rendererClasses [&NiagaraSpriteRendererProperties]
  renderers [
   {rendererClass=&NiagaraSpriteRendererProperties}
   {rendererClass=&NiagaraSpriteRendererProperties, rendererIndex=1}
  ]
 }
 Pulse_PDecal001 {
  emitterUpdateScript [
   &EmitterState(Life Cycle Mode=::NewEnumerator1, Loop Behavior=::NewEnumerator1, Loop Delay=0.8, Loop Duration=1, UseLoopDelay=true)
   &SpawnBurst_Instantaneous(Spawn Count=1, Use Spawn Probability=false)
  ]
  particleSpawnScript [
   /Niagara/Modules/Spawn/Initialization/V2/InitializeParticle(
    Color=<rgba 11.4558 13.5918 20 1>
    Lifetime=0.5
    Sprite Size Mode=::NewEnumerator3
    Uniform Sprite Size=100
    UsePositionOffset=false
    Write Material Random=true
   )
   &CameraOffset()
  ]
  particleUpdateScript [
   &ParticleState()
   &ScaleColor(
    Linear Color Curve=DataInterface() {propertyValues {
      AlphaCurve {keys [<at 0 0 RCIM_Cubic>, {arriveTangent=-1.373, interpMode=RCIM_Cubic, leaveTangent=-1.373, tangentMode=RCTM_User, time=0.3, value=1}, <at 1 0>]}
      BlueCurve {keys [<at 0 1>,<at 0.6831 1>]}
      GreenCurve {keys [<at 0 0.3636>,<at 0.6831 0>]}
      RedCurve {keys [<at 0 0>,<at 0.6831 0.4834>]}
     }}
    Scale Mode=::NewEnumerator2
   )
   &DynamicMaterialParameters(
    Index 0 Param 1=&FloatFromCurve(FloatCurve=DataInterface() {propertyValues {Curve {keys [<at 0 -1.1>,<at 0.5 1>]}}})
    Index 0 Param 2=1
    Index 0 Param 3=&FloatFromCurve(FloatCurve=DataInterface() {propertyValues {Curve {keys [{arriveTangent=1.3198, interpMode=RCIM_Cubic, leaveTangent=1.3198, tangentMode=RCTM_User, time=0, value=-0.25}]}}})
    Index 0 Param 4=1
    Module.Param0WriteEnabled=true
    Module.Param1WriteEnabled=true
    Module.Param2WriteEnabled=true
    Module.Param3WriteEnabled=true
   )
  ]
  rendererClasses [&NiagaraSpriteRendererProperties, &NiagaraMeshRendererProperties]
  renderers [
   {rendererClass=&NiagaraSpriteRendererProperties}
   {rendererClass=&NiagaraSpriteRendererProperties, rendererIndex=1}
   {rendererClass=&NiagaraMeshRendererProperties, rendererIndex=2}
  ]
 }
 Ring_Burst {
  emitterSpawnScript [/SetVariables(Emitter.UseParticleColorInput=1) {isSetParameters=true, module=SetVariables_#6}]
  emitterUpdateScript [
   &EmitterState(Life Cycle Mode=::NewEnumerator1, Loop Behavior=::NewEnumerator1, Loop Delay=0.8, Loop Duration=1, UseLoopDelay=true)
   &SpawnBurst_Instantaneous(Spawn Count=3, Use Spawn Probability=false)
   &SpawnBurst_Instantaneous(Spawn Count=3, Spawn Time=0.15, Use Spawn Probability=false) {module=SpawnBurst_Instantaneous001}
  ]
  particleSpawnScript [
   &InitializeParticle(
    Color=<rgba 0.9249 1 0.9295 1>
    Lifetime Max=0.5
    Lifetime Min=0.3
    Lifetime Mode=::NewEnumerator1
    Mesh Scale Mode=::NewEnumerator3
    Mesh Uniform Scale=0.05
    UsePositionOffset=false
    Write Material Random=true
   )
   &CameraOffset(Camera Offset Amount=8)
   /SetVariables(Particles.MeshOrientation=&MultiplyQuaternion(
     Quaternion A=&FindQuatBetweenTwoVectors(From Vector=<0 1 0>, To Vector=<0 0 1>)
     Quaternion B=&FindQuatBetweenTwoVectors(From Vector=<1 0 0>, To Vector=<0 1 0>)
    )) {isSetParameters=true}
   /SetVariables(Particles.Position=&RandomRangeVector(Minimum=<2.5 20 10>, Maximum=<2.5 60 10>)) {isSetParameters=true}
  ]
  particleUpdateScript [
   &ParticleState()
   /SetVariables(Particles.SizeWave:NiagaraFloat=&Waveform(
     Global Phase=$Particles.NormalizedAge
     "Waveform [1]"=::NewEnumerator0
     "[1] Phase Scale"=0
     "[1] Phase"=$Particles.NormalizedAge
     "[1] Frequency"=0.3
     "[1] Amplitude Min/Max"=<1 0>
     "[1] Bias"=1
     "[2] Phase"=$Emitter.Age
     "[3] Phase"=$Emitter.Age
    )) {isSetParameters=true, module=SetVariables_#7}
   /SetVariables(Particles.Desaturation=0.6) {isSetParameters=true, module=SetVariables_#8}
   /SetVariables(Particles.WaveColor:LinearColor=&MakeLinearColor(
     R:NiagaraFloat=&Waveform(
      Global Phase=$Particles.NormalizedAge
      "Waveform [1]"=::NewEnumerator0
      "[1] Phase"=$Particles.NormalizedAge
      "[1] Frequency"=1.5
      "[1] Amplitude Min/Max"=<1 0>
      "[1] Bias"=$Particles.Desaturation
      "[2] Phase"=$Emitter.Age
      "[3] Phase"=$Emitter.Age
     )
     G:NiagaraFloat=&Waveform(
      Global Phase=$Particles.NormalizedAge
      "Waveform [1]"=::NewEnumerator1
      "[1] Phase"=$Particles.NormalizedAge
      "[1] Frequency"=1
      "[1] Amplitude Min/Max"=<1 0>
      "[1] Scale"=2
      "[1] Bias"=$Particles.Desaturation
      "[2] Phase"=$Emitter.Age
      "[3] Phase"=$Emitter.Age
     )
     B:NiagaraFloat=&Waveform(
      Global Phase=$Particles.NormalizedAge
      "Waveform [1]"=::NewEnumerator2
      "[1] Phase Scale"=2.5
      "[1] Phase"=$Particles.NormalizedAge
      "[1] Frequency"=2
      "[1] Amplitude Min/Max"=<1 0>
      "[1] Bias"=$Particles.Desaturation
      "[2] Phase"=$Emitter.Age
      "[3] Phase"=$Emitter.Age
     )
     A:NiagaraFloat=1
    )) {isSetParameters=true, module=SetVariables_#9}
   /SetVariables(Particles.Color=&MultiplyLinearColor(
     Linear Color=$Particles.WaveColor
     Scale Factor=&MakeVector4_FromLinearColor(LinearColor=&MakeLinearColorFromVectorAndFloat(
       "Vector (RGB)"=<0.2 0.5 1>
       "Float (Alpha)"=&FloatFromCurve(FloatCurve=DataInterface() {propertyValues {Curve {keys [
            {arriveTangent=9.0725, interpMode=RCIM_Cubic, leaveTangent=9.0725, tangentMode=RCTM_User, time=0, value=-0.05}
            {arriveTangent=-2.5249, interpMode=RCIM_Cubic, leaveTangent=-2.5249, tangentMode=RCTM_User, time=0.4, value=1.3}
            {arriveTangent=-2.1667, interpMode=RCIM_Cubic, leaveTangent=-2.1667, time=0.8, value=0.2}
            {arriveTangent=-0.0405, interpMode=RCIM_Cubic, leaveTangent=-0.0405, tangentMode=RCTM_User, time=1, value=0}
           ]}}})
      ))
    )) {isSetParameters=true, module=SetVariables_#10}
   &ScaleMeshSize(Scale Factor=&Multiply_Vector(
     A=&MakeVector(X=$Particles.SizeWave, Y=$Particles.SizeWave, Z=1)
     B=&VectorFromFloat(Value=&FloatFromCurve(
       FloatCurve=DataInterface() {propertyValues {Curve {keys [
           {arriveTangent=4.4833, interpMode=RCIM_Cubic, leaveTangent=4.4833, tangentMode=RCTM_User, time=0, value=0.25}
           {arriveTangent=-0.7444, interpMode=RCIM_Cubic, leaveTangent=-0.7444, tangentMode=RCTM_User, time=0.2, value=0.95}
           {arriveTangent=-0.0405, interpMode=RCIM_Cubic, leaveTangent=-0.0405, tangentMode=RCTM_User, time=1, value=0.1}
          ]}}}
       Scale Curve=2
      ))
    )) {module=ScaleMeshSize001}
   &DynamicMaterialParameters(
    Index 0 Param 1=&FloatFromCurve(
     FloatCurve=DataInterface() {propertyValues {Curve {keys [
         {arriveTangent=9.0725, interpMode=RCIM_Cubic, leaveTangent=9.0725, tangentMode=RCTM_User, time=0, value=0}
         {arriveTangent=0.1068, interpMode=RCIM_Cubic, leaveTangent=0.1068, tangentMode=RCTM_User, time=0.1515, value=0.9243}
         {arriveTangent=-0.0405, interpMode=RCIM_Cubic, leaveTangent=-0.0405, tangentMode=RCTM_User, time=1, value=0}
        ]}}}
     Scale Curve=0.25
    )
    Module.Param0WriteEnabled=true
    Module.Param1WriteEnabled=true
    Module.Param2WriteEnabled=true
    Module.Param3WriteEnabled=true
   )
  ]
  rendererClasses [&NiagaraMeshRendererProperties]
  renderers [{rendererClass=&NiagaraMeshRendererProperties, rendererIndex=0}]
 }
 Ring_Warmup {
  emitterSpawnScript [/SetVariables() {isSetParameters=true, module=SetVariables_#6}]
  emitterUpdateScript [
   &EmitterState(Life Cycle Mode=::NewEnumerator1, Loop Behavior=::NewEnumerator1, Loop Duration=0.8, UseLoopDelay=false)
   &SpawnRate(
    SpawnRate=&FloatFromCurve(FloatCurve=DataInterface() {propertyValues {Curve {keys [<at 0 1>,<at 0.6 10>]}}}, CurveIndex=$Emitter.Age)
    Use Spawn Probability=false
   )
  ]
  particleSpawnScript [
   &InitializeParticle(
    Color=<rgba 0.9249 1 0.9295 1>
    Lifetime Max=0.6
    Lifetime Min=0.6
    Lifetime Mode=::NewEnumerator1
    Mesh Scale Mode=::NewEnumerator3
    Mesh Uniform Scale=0.05
    Sprite Size Mode=::NewEnumerator3
    Uniform Sprite Size=15
    UsePositionOffset=false
    Write Material Random=true
   )
   &CameraOffset(Camera Offset Amount=15)
   &SystemLocation(MaskSpawnGroup=false, Offset=<2.5 10 10>)
   &AddVelocity(Velocity=<0 120 0>)
  ]
  particleUpdateScript [
   &SolveForcesAndVelocity(Clamp Velocity=false, Limit Acceleration=false)
   &ParticleState()
   /SetVariables(Particles.SizeWave:NiagaraFloat=&Waveform(
     Global Phase=$Particles.NormalizedAge
     "Waveform [1]"=::NewEnumerator0
     "[1] Phase"=$Particles.NormalizedAge
     "[1] Frequency"=0.3
     "[1] Amplitude Min/Max"=<1 0>
     "[1] Scale"=1.5
     "[1] Bias"=1
     "[2] Phase"=$Emitter.Age
     "[3] Phase"=$Emitter.Age
    )) {isSetParameters=true, module=SetVariables_#7}
   /SetVariables(Particles.Desaturation=0.6) {isSetParameters=true, module=SetVariables_#8}
   /SetVariables(Particles.WaveColor:LinearColor=&MakeLinearColor(
     R:NiagaraFloat=&Waveform(
      Global Phase=$Particles.NormalizedAge
      "Waveform [1]"=::NewEnumerator0
      "[1] Phase"=$Particles.NormalizedAge
      "[1] Frequency"=1.5
      "[1] Amplitude Min/Max"=<1 0>
      "[1] Bias"=$Particles.Desaturation
      "[2] Phase"=$Emitter.Age
      "[3] Phase"=$Emitter.Age
     )
     G:NiagaraFloat=&Waveform(
      Global Phase=$Particles.NormalizedAge
      "Waveform [1]"=::NewEnumerator1
      "[1] Phase"=$Particles.NormalizedAge
      "[1] Frequency"=1
      "[1] Amplitude Min/Max"=<1 0>
      "[1] Scale"=2
      "[1] Bias"=$Particles.Desaturation
      "[2] Phase"=$Emitter.Age
      "[3] Phase"=$Emitter.Age
     )
     B:NiagaraFloat=&Waveform(
      Global Phase=$Particles.NormalizedAge
      "Waveform [1]"=::NewEnumerator2
      "[1] Phase Scale"=2.5
      "[1] Phase"=$Particles.NormalizedAge
      "[1] Frequency"=2
      "[1] Amplitude Min/Max"=<1 0>
      "[1] Bias"=$Particles.Desaturation
      "[2] Phase"=$Emitter.Age
      "[3] Phase"=$Emitter.Age
     )
     A:NiagaraFloat=1
    )) {isSetParameters=true, module=SetVariables_#9}
   /SetVariables(Particles.Color=&MultiplyLinearColor(
     Linear Color=$Particles.WaveColor
     Scale Factor=&MakeVector4_FromLinearColor(LinearColor=&MakeLinearColorFromVectorAndFloat(
       "Vector (RGB)"=<0.2 0.5 1>
       "Float (Alpha)"=&FloatFromCurve(
        FloatCurve=DataInterface() {propertyValues {Curve {keys [
            {arriveTangent=9.0725, interpMode=RCIM_Cubic, leaveTangent=9.0725, tangentMode=RCTM_User, time=0, value=1}
            {arriveTangent=-2.5249, interpMode=RCIM_Cubic, leaveTangent=-2.5249, tangentMode=RCTM_User, time=0.3, value=0.95}
            {arriveTangent=-1.3571, interpMode=RCIM_Cubic, leaveTangent=-1.3571, time=0.6, value=0.2}
            {arriveTangent=-0.0405, interpMode=RCIM_Cubic, leaveTangent=-0.0405, tangentMode=RCTM_User, time=1, value=0}
           ]}}}
        Scale Curve=$User.Opacity
       )
      ))
    )) {isSetParameters=true, module=SetVariables_#10}
   /SetVariables(Particles.ParticleDecalColor:LinearColor=&MultiplyLinearColor(
     Linear Color=$Particles.WaveColor
     Scale Factor=&MakeVector4_FromLinearColor(LinearColor=&MakeLinearColorFromVectorAndFloat(
       "Vector (RGB)"=<0.1 0.2 1>
       "Float (Alpha)"=&FloatFromCurve(
        FloatCurve=DataInterface() {propertyValues {Curve {keys [
            {arriveTangent=9.0725, interpMode=RCIM_Cubic, leaveTangent=9.0725, tangentMode=RCTM_User, time=0, value=1}
            {arriveTangent=-2.5249, interpMode=RCIM_Cubic, leaveTangent=-2.5249, tangentMode=RCTM_User, time=0.3, value=0.95}
            {arriveTangent=-1.3571, interpMode=RCIM_Cubic, leaveTangent=-1.3571, time=0.6, value=0.2}
            {arriveTangent=-0.0405, interpMode=RCIM_Cubic, leaveTangent=-0.0405, tangentMode=RCTM_User, time=1, value=0}
           ]}}}
        Scale Curve=$User.Opacity
       )
      ))
    )) {isSetParameters=true}
   /SetVariables(Particles.RingRotation:Quat4f=&MultiplyQuaternion(
     Quaternion A=&FindQuatBetweenTwoVectors(From Vector=<0 1 0>, To Vector=<0 0 1>)
     Quaternion B=&FindQuatBetweenTwoVectors(From Vector=<1 0 0>, To Vector=<0 1 0>)
    )) {isSetParameters=true}
   &ScaleMeshSize(Scale Factor=&Multiply_Vector(
     A=&MakeVector(X=$Particles.SizeWave, Y=$Particles.SizeWave, Z=1)
     B=&VectorFromFloat(Value=&FloatFromCurve(FloatCurve=DataInterface() {propertyValues {Curve {keys [
           {arriveTangent=4.4833, interpMode=RCIM_Cubic, leaveTangent=4.4833, tangentMode=RCTM_User, time=0, value=0.25}
           {arriveTangent=0.2072, interpMode=RCIM_Cubic, leaveTangent=0.2072, tangentMode=RCTM_User, time=0.5, value=0.85}
           {arriveTangent=-2.2491, interpMode=RCIM_Cubic, leaveTangent=-2.2491, tangentMode=RCTM_User, time=1, value=0.2758}
          ]}}}))
    )) {module=ScaleMeshSize001}
   &DynamicMaterialParameters(
    Index 0 Param 1=&FloatFromCurve(
     FloatCurve=DataInterface() {propertyValues {Curve {keys [
         {arriveTangent=9.0725, interpMode=RCIM_Cubic, leaveTangent=9.0725, tangentMode=RCTM_User, time=0, value=0}
         {arriveTangent=0.1068, interpMode=RCIM_Cubic, leaveTangent=0.1068, tangentMode=RCTM_User, time=0.1515, value=0.9243}
         {arriveTangent=-0.0405, interpMode=RCIM_Cubic, leaveTangent=-0.0405, tangentMode=RCTM_User, time=1, value=0}
        ]}}}
     Scale Curve=0.25
    )
    Module.Param0WriteEnabled=true
    Module.Param1WriteEnabled=true
    Module.Param2WriteEnabled=true
    Module.Param3WriteEnabled=true
   )
  ]
  rendererClasses [&NiagaraMeshRendererProperties, &NiagaraSpriteRendererProperties]
  renderers [
   {rendererClass=&NiagaraMeshRendererProperties, rendererIndex=0}
   {rendererClass=&NiagaraSpriteRendererProperties, rendererIndex=1}
   {rendererClass=&NiagaraSpriteRendererProperties, rendererIndex=2}
  ]
 }
 Sparkles_Burst {
  emitterUpdateScript [
   &EmitterState(Life Cycle Mode=::NewEnumerator1, Loop Behavior=::NewEnumerator1, Loop Delay=0.8, Loop Duration=0.2, UseLoopDelay=true)
   &SpawnRate(SpawnRate=40, Use Spawn Probability=false)
  ]
  particleSpawnScript [
   /Niagara/Modules/Spawn/Initialization/InitializeParticle(
    Color=<rgba 0 0.2857 2 1>
    Lifetime=0.2
    Module.Write Color=true
    Module.Write Lifetime=true
    Module.Write Mass=true
    Module.Write Material Random=true
    Module.Write Position=false
    Module.Write Scale=false
    Module.Write SpriteRotation=true
    Module.Write SpriteSize=true
    Position=&SimulationPosition()
    Sprite Rotation=&RandomRangeFloat(Minimum=0, Maximum=360)
    Sprite Size=&Vector2DFromFloat(Value=&RandomRangeFloat(Minimum=1, Maximum=2))
   )
   &CameraOffset(Camera Offset Amount=20)
   &SystemLocation(MaskSpawnGroup=false, Offset=<0 40 10>)
   &SphereLocation(
    MaskSpawnGroup=false
    Non Uniform Scale=<1 2 1>
    Override Seed=false
    Sphere Radius=20
    Surface Only=false
    UseNonUniformScale=true
    UseOffset=false
    UseSphereOrientation=false
   )
  ]
  particleUpdateScript [
   &ParticleState()
   &ScaleSpriteSize(Scale Factor=&Vector2DFromFloat(Value=&RandomRangeFloat(Minimum=3.5, Maximum=5, Evaluation Type=::NewEnumerator1)))
   &ScaleColor(Scale Alpha=$User.Opacity, ScaleA=true, ScaleRGB=true)
  ]
  rendererClasses [&NiagaraSpriteRendererProperties]
  renderers [
   {rendererClass=&NiagaraSpriteRendererProperties}
   {rendererClass=&NiagaraSpriteRendererProperties, rendererIndex=1}
  ]
 }
 Spirals_Large {
  bEnabled=false
  simTarget=GPUComputeSim
  emitterUpdateScript [
   &EmitterState(Life Cycle Mode=::NewEnumerator1, Loop Behavior=::NewEnumerator1, Loop Delay=0.2, Loop Duration=0.4, UseLoopDelay=true)
   &SpawnRate(SpawnRate=40, Use Spawn Probability=false)
  ]
  particleSpawnScript [
   /Niagara/Modules/Spawn/Initialization/V2/InitializeParticle(
    Color=<rgba 0 10.085 20 1>
    Lifetime=0.6
    Sprite Size Mode=::NewEnumerator3
    Uniform Sprite Size=&RandomRangeFloat(Minimum=2, Maximum=4)
    UsePositionOffset=false
    Write Material Random=true
   )
   &SystemLocation(MaskSpawnGroup=false, Offset=<0 0 50>)
  ]
  particleUpdateScript [
   &ParticleState()
   &RotateAroundPoint(
    Radius:NiagaraFloat=&RandomRangeFloat(Minimum=5, Maximum=10)
    Rotation Center:NiagaraPosition=&MakeVector(
     X=1
     Y=&FloatFromCurve(FloatCurve=DataInterface() {propertyValues {Curve {keys [<at 0 20>,<at 0.6 80>]}}}, CurveIndex=$Particles.Age)
     Z=10
    )
    Rotation Phase:NiagaraFloat=&RandomRangeFloat(Minimum=0, Maximum=0.1)
    Rotation Rate=2
    Y Axis:Vector3f=<0 0 1>
   )
   &SolveForcesAndVelocity(Clamp Velocity=false, Limit Acceleration=false)
   &ScaleColor(
    Scale Alpha=&FloatFromCurve(FloatCurve=DataInterface() {propertyValues {Curve {keys [<at 0 0>,<at 0.25 1>,<at 0.6 1>,<at 0.9 0>]}}})
    ScaleA=true
    ScaleRGB=false
   )
   &ScaleSpriteSize(Scale Factor=&Vector2DFromFloat(Value=&FloatFromCurve(FloatCurve=DataInterface() {propertyValues {Curve {keys [<at 0 0>,<at 0.2 1>,<at 0.6 1>,<at 1 0>]}}})))
  ]
  rendererClasses [&NiagaraSpriteRendererProperties]
  renderers [
   {rendererClass=&NiagaraSpriteRendererProperties}
   {rendererClass=&NiagaraSpriteRendererProperties, rendererIndex=1}
  ]
 }
 Spirals_Small {
  bEnabled=false
  simTarget=GPUComputeSim
  emitterUpdateScript [
   &EmitterState(Life Cycle Mode=::NewEnumerator1, Loop Behavior=::NewEnumerator1, Loop Delay=0.2, Loop Duration=0.4, UseLoopDelay=true)
   &SpawnRate(SpawnRate=40, Use Spawn Probability=false)
  ]
  particleSpawnScript [
   &InitializeParticle(
    Color=<rgba 0 5.0425 10 1>
    Lifetime=0.6
    Sprite Size Mode=::NewEnumerator3
    Uniform Sprite Size=&RandomRangeFloat(Minimum=1, Maximum=2)
    UsePositionOffset=false
    Write Material Random=true
   )
   &SystemLocation(MaskSpawnGroup=false, Offset=<0 0 50>)
  ]
  particleUpdateScript [
   &ParticleState()
   &RotateAroundPoint(
    Radius:NiagaraFloat=&RandomRangeFloat(Minimum=2.5, Maximum=10)
    Rotation Center:NiagaraPosition=&MakeVector(
     X=1
     Y=&FloatFromCurve(FloatCurve=DataInterface() {propertyValues {Curve {keys [<at 0 20>,<at 0.6 80>]}}}, CurveIndex=$Particles.Age)
     Z=10
    )
    Rotation Phase:NiagaraFloat=&RandomRangeFloat(Minimum=-0.25, Maximum=0.25)
    Rotation Rate=2
    Y Axis:Vector3f=<0 0 1>
   )
   &SolveForcesAndVelocity(Clamp Velocity=false, Limit Acceleration=false)
   &ScaleColor(
    Scale Alpha=&FloatFromCurve(FloatCurve=DataInterface() {propertyValues {Curve {keys [<at 0 0>,<at 0.25 1>,<at 0.6 1>,<at 0.9 0>]}}})
    ScaleA=true
    ScaleRGB=false
   )
  ]
  rendererClasses [&NiagaraSpriteRendererProperties]
  renderers [
   {rendererClass=&NiagaraSpriteRendererProperties}
   {rendererClass=&NiagaraSpriteRendererProperties, rendererIndex=1}
  ]
 }
 Weapon_Mesh {
  emitterUpdateScript [
   &EmitterState(
    Inactive Response=::NewEnumerator1
    Life Cycle Mode=::NewEnumerator1
    Loop Behavior=::NewEnumerator1
    Loop Delay=0.2
    Loop Duration Mode=::NewEnumerator1
    UseLoopDelay=true
   )
   &SpawnBurst_Instantaneous(Spawn Count=1, Use Spawn Probability=false)
  ]
  particleSpawnScript [&InitializeParticle(Color=<rgba 0 10.085 20 1>, Lifetime=0.8, UsePositionOffset=false, Write Material Random=true)]
  particleUpdateScript [
   &ParticleState()
   &DynamicMaterialParameters(
    Index 0 Param 1=&FloatFromCurve(FloatCurve=DataInterface() {propertyValues {Curve {keys [<at 0 0>,<at 0.6 1>]}}}, CurveIndex=$Particles.Age)
    Index 0 Param 2=1.5
    Module.Param0WriteEnabled=true
    Module.Param1WriteEnabled=true
    Module.Param2WriteEnabled=true
    Module.Param3WriteEnabled=true
   )
  ]
  rendererClasses [&NiagaraMeshRendererProperties]
  renderers [{rendererClass=&NiagaraMeshRendererProperties, rendererIndex=0}]
 }
}
summary {
 systemName=NS_Skyfire_Weapon_Materialize
 emitters [
  {emitterName=Ring_Warmup, rendererClasses [&NiagaraMeshRendererProperties, &NiagaraSpriteRendererProperties]}
  {emitterName=Flare_Burst, rendererClasses [&NiagaraSpriteRendererProperties, &NiagaraLightRendererProperties]}
  {emitterName=Ring_Burst, rendererClasses [&NiagaraMeshRendererProperties]}
  {bEnabled=false, emitterName=Flare_Warmup, rendererClasses [&NiagaraSpriteRendererProperties]}
  {emitterName=NE_Glow002, rendererClasses [&NiagaraSpriteRendererProperties]}
  {emitterName=Sparkles_Burst, rendererClasses [&NiagaraSpriteRendererProperties]}
  {emitterName=Decal_Warmup001, rendererClasses [&NiagaraMeshRendererProperties]}
  {
   emitterName=Pulse_PDecal001
   rendererClasses [&NiagaraSpriteRendererProperties, &NiagaraMeshRendererProperties]
  }
  {emitterName=Weapon_Mesh, rendererClasses [&NiagaraMeshRendererProperties]}
  {emitterName=AirGlow001, rendererClasses [&NiagaraSpriteRendererProperties]}
  {bEnabled=false, emitterName=Spirals_Large, simTarget=GPUComputeSim, rendererClasses [&NiagaraSpriteRendererProperties]}
  {bEnabled=false, emitterName=Spirals_Small, simTarget=GPUComputeSim, rendererClasses [&NiagaraSpriteRendererProperties]}
  {emitterName=Light, rendererClasses [&NiagaraSpriteRendererProperties, &NiagaraLightRendererProperties]}
  {emitterName=DustFan, rendererClasses [&NiagaraMeshRendererProperties]}
  {emitterName=DustFan001, rendererClasses [&NiagaraMeshRendererProperties]}
 ]
 userVariables [
  User.Opacity {type=NiagaraFloat, defaultValue=1, description=""}
  User.SkeletalMesh {
   type=NiagaraDataInterfaceSkeletalMesh
   defaultValue {
    struct=NiagaraExt_VariableValue_DataInterface
    value {dataInterface=&self:NiagaraDataInterfaceSkeletalMesh_0}
   }
  }
 ]
}
